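# Translucent UI material for the backing panel of an XP accolade in the Battle Royale HUD.
# An angled-box shape comes from optionally X-flipped UVs run through symmetry and 9-slice functions.
# Color switches or animates between an indexed adaptive palette with white mix and either a custom color or red and orange-yellow gradients picked by IsKnockedColor.
# Static switches choose animated or fixed color and opacity, with opacity multiplied by a rerouted mask.
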Material /BattleRoyaleGameUI/HUD/XP/Assets/M_UI_Accolade_Backing
graph {
 properties {bAutomaticallySetUsageInEditor=false, blendMode=BLEND_Translucent, materialDomain=MD_UI, usedShadingModels=Default Lit}
 nodes [
  n1 = MaterialFunctionCall!(BufferSource) {materialFunction=/BattleRoyaleGameUI/Shared/MF_UI_AdaptiveColors}
  n2 = ScalarParameter() {group=Color, parameterName=AdaptiveColorIndex}
  n3 = Subtract(A=@n2)
  n4 = LinearInterpolate(
   A=LinearInterpolate(A=LinearInterpolate(A=@n1.Dark, B=@n1.Dark, Alpha=@n2), B=@n1.Medium, Alpha=Saturate(None=@n3))
   B=@n1.Bright
   Alpha=Saturate(None=Subtract(A=@n2) {constB=2})
  )
  n5 = Saturate(None=Subtract(A=@n2) {constB=3})
  NamedRerouteDeclaration(None=LinearInterpolate(A=LinearInterpolate(A=@n4, B=@n1.White, Alpha=@n5), B=@n1.White, Alpha=ScalarParameter() {parameterName=WhiteMix})) {name=AdaptiveColorOutput, nodeColor=<rgba 0 0 0 1>}
  n6 = MaterialFunctionCall() {materialFunction=&GetUserInterfaceUV}
  n7 = MaterialFunctionCall(Float2=@n6.Normalized UV) {materialFunction=&BreakOutFloat2Components}
  n8 = OneMinus(None=@n7.R)
  n9 = MaterialFunctionCall!(
   UV=LinearInterpolate(A=@n6.Normalized UV, B=AppendVector(A=@n8, B=@n7.G), Alpha=ScalarParameter() {parameterName=FlipX})
   Mirror Y
   Swizzle UV
  ) {materialFunction=/FortUILibrary/MaterialFunctions/Basics/MF_UI_Symmetry_2x1_New}
  NamedRerouteDeclaration(None=MaterialFunctionCall!(
    UV=MaterialFunctionCall!(UVs=@n9.Flipped Symmetry) {materialFunction=/FortUILibrary/MaterialFunctions/Basics/MF_UI_9SliceUVs}.UVs
    Corner=Constant() {r=1}
    Angle=ScalarParameter() {defaultValue=24, parameterName=Angle}
    Flip
    Falloff
   ) {materialFunction=/FortUILibrary/MaterialFunctions/Shapes/MF_UI_BoxAngle}.Result) {name=Shape, nodeColor=<rgba 0 0 0 1>}
  n10 = NamedRerouteUsage()
  n11 = Constant4Vector() {constant=<rgba 0.5 0 0.0667 1>}
  n12 = NamedRerouteUsage()
  n13 = LinearInterpolate(
   A=@n11.RGB
   B=ComponentMask(None=CollectionParameter() {collection=/BattleRoyaleGameUI/Shared/MPC_BRGameUI, parameterName=Red}) {b=true, g=true, r=true}
   Alpha=@n12
  )
  n14 = LinearInterpolate(A=@n11.RGB, B=@n13, Alpha=NamedRerouteUsage())
  n15 = ScalarParameter() {parameterName=Fade}
  n16 = ComponentMask(None=CollectionParameter() {collection=&MPC_BRGameUI, parameterName=Orange}) {b=true, g=true, r=true}
  n17 = LinearInterpolate(
   A=@n16
   B=ComponentMask(None=CollectionParameter() {collection=&MPC_BRGameUI, parameterName=Yellow}) {b=true, g=true, r=true}
   Alpha=LinearInterpolate(Alpha=@n12) {constA=0.2, constB=0.5}
  )
  n18 = LinearInterpolate(A=@n17, B=@n16, Alpha=NamedRerouteUsage())
  n19 = ScalarParameter() {parameterName=Fade}
  n20 = StaticSwitchParameter(
   True=VectorParameter() {group=Color, parameterName=Color, defaultValue=<rgba 1 0 0 1>}.RGB
   False=LinearInterpolate(
    A=LinearInterpolate(A=@n13, B=@n14, Alpha=@n15)
    B=LinearInterpolate(A=@n17, B=@n18, Alpha=@n19)
    Alpha=ScalarParameter() {parameterName=IsKnockedColor}
   )
  ) {parameterName=CustomColor}
  n21 = ScalarParameter() {group=Color, parameterName=Anim_Color}
  NamedRerouteDeclaration(None=StaticSwitchParameter(
    True=LinearInterpolate(A=@n10, B=@n20, Alpha=@n21)
    False=StaticSwitchParameter(True=@n10, False=@n20) {defaultValue=true, parameterName=AdaptiveColor}
   ) {defaultValue=true, parameterName=AnimatedColors}) {name=Color, nodeColor=<rgba 0 0 0 1>}
  n22 = LinearInterpolate(
   A=ScalarParameter() {group=Opacity, parameterName=Opacity_Start}
   B=ScalarParameter() {defaultValue=1, group=Opacity, parameterName=Opacity_Stop}
   Alpha=ScalarParameter() {defaultValue=1, group=Opacity, parameterName=Anim_Opacity}
  )
  n23 = ScalarParameter() {defaultValue=1, group=Opacity, parameterName=Opacity}
  NamedRerouteDeclaration(None=Multiply(A=StaticSwitchParameter(True=@n22, False=@n23) {defaultValue=true, parameterName=AnimatedOpacity}, B=NamedRerouteUsage())) {name=Opacity, nodeColor=<rgba 0 0 0 1>}
  NamedRerouteDeclaration(None=ComponentMask(None=@n6.Normalized UV) {g=true}) {name=VerticalMask, nodeColor=<rgba 0 0 0 1>}
  NamedRerouteDeclaration(None=ComponentMask(None=@n6.Normalized UV) {r=true}) {name=HorizontalMask, nodeColor=<rgba 0 0 0 1>}
  NamedRerouteUsage()
  NamedRerouteUsage()
 ]
}
